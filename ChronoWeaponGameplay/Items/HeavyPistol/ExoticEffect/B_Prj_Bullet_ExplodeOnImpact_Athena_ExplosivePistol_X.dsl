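; Explosive exotic Heavy Pistol bullet projectile that explodes on impact, with a material-animated fake tracer and Niagara trail.
; On stop it calls DoExplosion, plays cosmetic sound, radial force and decal, and sends a gameplay event if predicted.
; Setting HitLocation drives OnRep_HitLocation for radius-scaled explosion FX, surface FX, feedback cue and sound indicator.
; Water entry spawns splash FX at the hit from the otherwise empty OnTouched, then destroys the actor.
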(event EventBeginPlay
  (bind _returnvalue (Utilities|Time|GetGameTimeinSeconds))
  (bind _tracermesh (Variables|Default|GetTracerMesh))
  ;[1 1 1] Animate fake tracer mesh
  (Networking|CanExecuteCosmeticEvents
    (:True
      (bind _asfort_player_pawn (Utilities|Casting|CastToFortPlayerPawn (Variables|Actor|GetInstigator)))
      (Utilities|FlowControl|Sequence
        (:then_0
          (Rendering|Material|SetVectorParameterValueonMaterials _tracermesh "BulletOrigin" (Effects|GetMuzzleLocation (Variables|/Script/FortniteGame.FortPawn|GetCurrentWeapon _asfort_player_pawn) -1))
          (Rendering|Material|SetScalarParameterValueonMaterials _tracermesh "StartTime" _returnvalue)
          (Rendering|Material|SetScalarParameterValueonMaterials _tracermesh "WPODuration" (Math|Float|MapRangeClamped (Math|Vector|VectorLength (Transformation|GetVelocity)) 5e+03 4e+04 0.1 0.25)))
        (:then_1
          (Effects|Components|ParticleSystem|SetVectorParameter (Variables|Default|GetNS_Sniper_Bullet_Trail_Fiery) "BulletOrigin" (Effects|GetMuzzleLocation (Variables|/Script/FortniteGame.FortPawn|GetCurrentWeapon _asfort_player_pawn) -1))
          (Rendering|Material|SetScalarParameterforCustomPrimitiveData (Variables|Default|GetNS_Sniper_Bullet_Trail_Fiery) "StartTime" _returnvalue))))))

(event EventOnTouched
  (params (OtherActor :type Actor%20Object%20Reference) (OtherComp :type Primitive%20Component%20Object%20Reference) (HitResult :type Hit%20Result%20Structure%20%28by%20ref%29) (bIsOverlap :type Boolean)))

(event Custom|Spawn%20Water%20FXs
  (bind _waterinteractioncomponent (Variables|Water|GetWaterInteractionComponent))
  (bind (_waterplanelocation _waterplanenormal _watersurfaceposition _waterdepth _waterbodyidx) (FX|GetLastWaterSurfaceInfo _waterinteractioncomponent :!class /Script/FortniteGame.FortWaterInteractionComponent))
  ;[1 1 1] EVENT -> OnTouched
  (bind _returnvalue (Effects|Components|ParticleSystem|SpawnAltFXAtLocation "WeaponImpactEffect" (Variables|Default|GetNS%20Water%20Explosion%20FX) 0 (.Location (Collision|BreakHitResult (stale EventOnTouched HitResult))) (Math|Conversions|RotationFromXVector _waterplanenormal)))
  (Effects|Components|ParticleSystem|SetFloatParameter _returnvalue "WaterDepth" _waterdepth)
  (Effects|Components|ParticleSystem|SetFloatParameter _returnvalue "WaterBodyIndex" (Math|Conversions|ToFloat(Integer) _waterbodyidx))
  (Default|SpawnFNRadialForce (Math|Transform|MakeTransform (.Location (Collision|BreakHitResult (stale EventOnTouched HitResult)))) (Variables|Default|GetPlayrate) (Variables|Default|GetImpact%20Radius) (Variables|Default|GetStrength) (Variables|Default|GetTime%20Decay))
  (Actor|DestroyActor))

(event OnEnterWaterEvent(WaterInteractionComponent)
  (params (WaterBody :type Fort%20Water%20Body%20Actor%20Object%20Reference) (WaterInteractionComponent :type Fort%20Water%20Interaction%20Component%20Object%20Reference) (bIsFirstBody :type Boolean))
  (bind _sound (Variables|Default|GetSound))
  (Utilities|FlowControl|Sequence
    (:then_0
      (Utilities|IsValid _sound
        (:Is%20Valid
          (if (Audio|Components|Audio|IsPlaying (Variables|Default|GetExplosionAudio))
            ;[1 1 1] EVENT -> OnTouched
            (Audio|Components|Audio|FadeOut (Variables|Default|GetExplosionAudio) 0.1 0.0 "Linear" :!class /Script/Engine.AudioComponent))
          (bind _returnvalue (Audio|SpawnSoundatLocation _sound (Transformation|GetActorLocation)))
          (Variables|Default|SetExplosionAudio _returnvalue))))
    (:then_1
      (Utilities|FlowControl|DoOnce false
        (:Completed
          (if (Utilities|IsValid (Variables|Default|GetNS%20Water%20Explosion%20FX) :!class /Script/Engine.KismetSystemLibrary)
            (CallFunction|SpawnWaterFXs)))))))

(event EventOnStop
  (params (Hit :type Hit%20Result%20Structure%20%28by%20ref%29))
  (bind (_bblockinghit _binitialoverlap _time _distance _location _impactpoint _normal _impactnormal _physmat _hitactor) (Collision|BreakHitResult Hit))
  (bind _returnvalue (Game|GetInstigator))
  (bind (_returnvalue_roll _returnvalue_pitch _returnvalue_yaw) (Math|Rotator|BreakRotator (Math|Conversions|RotationFromXVector _impactnormal)))
  (bind (_asfort_pawn _bsuccess) (Utilities|Casting|CastToFortPawn _hitactor))
  (bind _ns_sniper_bullet_trail_fiery (Variables|Default|GetNS_Sniper_Bullet_Trail_Fiery))
  (bind _surfacetype (Variables|/Script/PhysicsCore.PhysicalMaterial|GetSurfaceType _physmat))
  (bind _output_get (Variables|Default|SetCachedHit Hit))
  (Utilities|FlowControl|Sequence
    (:then_0
      (Variables|/Script/Engine.ProjectileMovementComponent|SetProjectileGravityScale 0.0 (Variables|Movement|GetProjectileMovementComponent))
      (Class|BulletWhipTrackerComponentBase|Disable (Variables|Visual|GetBulletWhipTrackerComponent)))
    (:then_1
      (Networking|CanExecuteCosmeticEvents
        (:True
          (Variables|BPrjBoomBow|SetHitLocation _location :!pinned true) ; All Explosion VFX
          (Variables|BPrjBoomBow|SetHitRotation (Math|Rotator|MakeRotfromX _impactnormal))
          (Variables|Default|SetSurface%20Type _surfacetype)
          (Variables|BPrjBoomBow|SetHitPlayer _bsuccess)
          (CallFunction|PlayImpactSFX _surfacetype _hitactor _location)
          (Default|SpawnFNRadialForce (Math|Transform|MakeTransform _location) (Variables|Default|GetPlayrate) (Variables|Default|GetImpact%20Radius) (Variables|Default|GetStrength) (Variables|Default|GetTime%20Decay))
          (Transformation|DetachFromComponent _ns_sniper_bullet_trail_fiery "KeepWorld" "KeepWorld" "KeepWorld" true)
          (Utilities|FlowControl|Sequence
            (:then_0
              (Components|Activation|Deactivate (Variables|Default|GetNS_Sniper_Bullet_Trail_Fiery))
              (Utilities|FlowControl|Delay 0.2)
              (Components|Activation|Deactivate _ns_sniper_bullet_trail_fiery))
            (:then_1
              (if (not (or (Fortnite|IsMobilePlatform) _bsuccess))
                (bind _returnvalue_2 (Rendering|Decal|SpawnDecalatLocation "/ChronoWeaponGameplay/Items/HeavyPistol/FX/MD_Decal_BulletHole_Lifetime_Inst.MD_Decal_BulletHole_Lifetime_Inst" (Variables|Default|GetDecalSize) _location (Math|Rotator|MakeRotator (+ _returnvalue_roll (Math|Random|RandomFloatinRange 0.0 3.6e+02)) _returnvalue_pitch _returnvalue_yaw) 4.5))
                (Rendering|Components|Decal|SetFadeOut _returnvalue_2 0.0 4.5 true))
              (Utilities|FlowControl|Delay 0.2)
              (Rendering|SetVisibility (Variables|Default|GetTracerMesh) false false))))))
    (:then_2
      (Delivery|DoExplosion))
    (:then_3
      (if (Variables|Default|GetIsPredicted)
        (bind _asfort_player_pawn (Utilities|Casting|CastToFortPlayerPawn _returnvalue))
        (Variables|Default|SetDamageContext (Class|AthenaFunctionLibrary|MakeSimpleFortEffectContext _returnvalue))
        (Ability|EffectContext|AddHitResult (Variables|Default|GetDamageContext) _output_get)
        (Ability|SendGameplayEventtoActor _asfort_player_pawn "(TagName=\"Ability.DMR.Fire\")" (Utilities|Struct|MakeGameplayEventData "(TagName=\"Ability.DMR.Fire\")" _asfort_player_pawn 0 0 0 (Variables|Default|GetDamageContext) "(GameplayTags=)" "(GameplayTags=)" 0.0 (Ability|TargetData|AbilityTargetDatafromActorArray (Utilities|Array|MakeArray (.HitActor (Collision|BreakHitResult _output_get))))))))))

(fn BlastRadius
  (Variables|Default|SetBlastRadiusScalar (/ (ScalableFloat|GetValueatLevel (Variables|Default|GetExplosionRadius)) 225.0)))

(fn OnRep_HitLocation
  (bind _surface_type (Variables|Default|GetSurface%20Type))
  (bind _returnvalue (Math|Conversions|ToInteger(Byte) _surface_type))
  (bind _hitrotation (Variables|BPrjBoomBow|GetHitRotation))
  (bind _hitlocation (Variables|BPrjBoomBow|GetHitLocation))
  (CallFunction|BlastRadius)
  (bind _returnvalue_2 (Niagara|SpawnSystematLocation (Variables|Default|GetNS_Explosion) _hitlocation _hitrotation))
  (Niagara|SetNiagaraVariable(Float) _returnvalue_2 "BlastRadiusScalar" (Variables|Default|GetBlastRadiusScalar))
  ;[1 1 1] PhysSurface Interaction for explosion Niagara FX
  (bind _checkedphysicalsurfaceid (CallFunction|PhysicalSurfaceCheck _returnvalue))
  (if _checkedphysicalsurfaceid
    (bind _returnvalue_3 (Effects|Components|ParticleSystem|SpawnFXAtLocation (Variables|Default|GetFX_Physical%20Surface) _hitlocation _hitrotation))
    (Effects|Components|ParticleSystem|SetIntParameter _returnvalue_3 "PhysicalSurface" _returnvalue)
    (Effects|Components|ParticleSystem|SetFloatParameter _returnvalue_3 "Global Scale" 0.6)
    (else
      _))
  (Rendering|SetRenderCustomDepth _returnvalue_2 true)
  (LOD|SetMaxDrawDistance _returnvalue_2 (Variables|Default|GetDrawDistance))
  (Components|Activation|Deactivate (Variables|Default|GetNS_Sniper_Bullet_Trail_Fiery))
  (Utilities|FlowControl|Sequence
    (:then_0
      (GameplayCue|ExecuteGameplayCue%28Burst%29Local self (Variables|Default|GetFeedbackCue) (Ability|GameplayCue|MakeGameplayCueParameters 0.0)) ; Cam shake + controller vibration
    )
    (:then_1
      (Fortnite|ShowSoundIndicatorLocalforTeam :Location (Variables|BPrjBoomBow|GetHitLocation) :MaxAudibleDistance (Variables|/Script/Engine.SoundBase|GetMaxDistance (Variables|Sounds|GetCue_DistantSound)) :IndicatorType "Gunshot" :Team (Team|GetActorTeam self) :Affiliations (Utilities|Array|MakeArray "Hostile" "Neutral") :Instigator self :Tint "(R=0.000000,G=0.000000,B=0.000000,A=0.000000)" :AttenuationMaxRangeMultiplier 1.0 :AttenuationIntensityMultiplier 1.0) ; Visual Sound Indicator
    )))

(fn PhysicalSurfaceCheck
  (params (Physical%20Surface%20ID :type Integer))
  (return (Math|Boolean|ORBoolean (== Physical%20Surface%20ID 30) (== Physical%20Surface%20ID 26) (== Physical%20Surface%20ID 23) (== Physical%20Surface%20ID 24))))

(fn PlayImpactSFX
  (params (Surface%20Type :type EPhysicalSurface%20Enum :default "SurfaceType_Default") (Actor :type Actor%20Object%20Reference) (Location :type Vector))
  (if (Audio|Components|Audio|IsPlaying (Variables|Default|GetExplosionAudio))
    (Audio|Components|Audio|FadeOut (Variables|Default|GetExplosionAudio) 0.1 0.0 "Linear" :!class /Script/Engine.AudioComponent))
  (bind _returnvalue (Audio|SpawnSoundatLocation (Variables|Default|GetSound) Location))
  (Variables|Default|SetExplosionAudio _returnvalue))

(fn ConstructionScript
  (LOD|SetMaxDrawDistance (Variables|Default|GetTracerMesh) (Variables|Default|GetDrawDistance)))
